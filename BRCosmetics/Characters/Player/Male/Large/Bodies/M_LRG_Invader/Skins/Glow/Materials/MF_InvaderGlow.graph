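# Material function that overlays an animated lava-under-glass glow on the Invader male large body skin, blended onto input material attributes by an Alpha input.
# Lava is scrolling triplanar noise in pre-skinned space, remapped through a gradient ramp and faded by view angle.
# Fresnel terms drive glass color and mask.
# One RGBA mask supplies glass edges, distortion, gradient offsets and coverage for emissive, base color, metallic and roughness changes.
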MaterialFunction /BRCosmetics/Characters/Player/Male/Large/Bodies/M_LRG_Invader/Skins/Glow/Materials/MF_InvaderGlow
graph {nodes [
  n1 = FunctionInput() {inputName=Material, inputType=FunctionInput_MaterialAttributes}
  n2 = GetMaterialAttributes(None=@n1) {attributeGetTypes [#1]}
  n3 = ScalarParameter() {defaultValue=1, group=InvaderGlow, parameterName=RoughnessMultiplier, sortPriority=10}
  n4 = TextureSampleParameter2D() {
   group=InvaderGlowTextureMask
   parameterName=InvaderGlow_Mask
   samplerType=SAMPLERTYPE_Masks
   sortPriority=0
   texture=&T_BlackMask
  }
  n5 = NamedRerouteUsage()
  n6 = SetMaterialAttributes!(
   MaterialAttributes=@n1
   EmissiveColor=NamedRerouteUsage()
   Roughness=LinearInterpolate(B=Multiply(A=@n2.Roughness, B=@n3), Alpha=@n4.A) {constA=1}
   BaseColor=LinearInterpolate(B=NamedRerouteUsage(), Alpha=@n5) {constB=0}
   Metallic=LinearInterpolate(B=ScalarParameter() {group=InvaderGlow, parameterName=Metalness, sortPriority=11}, Alpha=@n5) {constB=0}
  ) {attributeSetTypes [
    #2
    #1
    #3
    #4
   ]}
  FunctionOutput(None=BlendMaterialAttributes(A=@n1, B=@n6, Alpha=FunctionInput() {inputName=Alpha, inputType=FunctionInput_Scalar}) {vertexAttributeBlendType=UseA})
  n7 = Subtract(A=CameraPositionWS(), B=WorldPosition().XYZ)
  n8 = Saturate(None=DotProduct(A=Normalize(VectorInput=@n7), B=PixelNormalWS()))
  n9 = SmoothStep(
   Min=ScalarParameter() {group=InvaderGlowLava, parameterName=LavaFalloffLow, sortPriority=40}
   Max=ScalarParameter() {defaultValue=1, group=InvaderGlowLava, parameterName=LavaFalloffHigh, sortPriority=41}
   Value=@n8
  )
  n10 = VectorParameter() {group=InvaderGlowLava, parameterName=LavaColor, sortPriority=10, defaultValue=<rgba 1 1 1 1>}
  n11 = ScalarParameter() {defaultValue=1, group=InvaderGlowLava, parameterName=LavaGradientTiling, sortPriority=20}
  n12 = Divide(
   A=VertexInterpolator(VS=PreSkinnedPosition()).PS
   B=ScalarParameter() {defaultValue=40, group=InvaderGlowLava, parameterName=LavaGradientScaleDivisor, sortPriority=21}
  )
  n13 = Multiply(A=@n11, B=@n12)
  n14 = Frac(None=Multiply(A=Time(), B=VectorParameter() {group=InvaderGlowLava, parameterName=LavaSpeed, sortPriority=13, defaultValue=<rgba 0 0.2 0.2 1>}.RGB))
  n15 = Multiply(A=ScalarParameter() {group=InvaderGlowTextureMask, parameterName=G_LavaDistortionAmount, sortPriority=20}, B=@n4.G)
  n16 = Add(A=ComponentMask(None=Add(A=@n13, B=@n14)) {b=true, g=true}, B=@n15)
  n17 = TextureObjectParameter() {
   group=InvaderGlowLava
   parameterName=LavaGradient
   samplerType=SAMPLERTYPE_LinearColor
   sortPriority=11
   texture=/BRCosmetics/Characters/Player/Male/Large/Bodies/M_LRG_Invader/Skins/Glow/Textures/T_InvaderGlow_Lava_Gradient
  }
  n18 = Multiply(A=@n11, B=@n12)
  n19 = Add(A=ComponentMask(None=Add(A=@n18, B=@n14)) {b=true, r=true}, B=@n15)
  n20 = VertexInterpolator(VS=PreSkinnedNormal())
  n21 = ComponentMask(None=@n20.PS) {g=true}
  n22 = ScalarParameter() {defaultValue=1, group=InvaderGlowLava, parameterName=LavaAxisBlendStrength, sortPriority=50}
  n23 = ComponentMask(None=MaterialFunctionCall(In=Abs(None=@n21), Contrast=@n22) {materialFunction=&CheapContrast}.Result) {g=true}
  n24 = ComponentMask(None=Add(A=Multiply(A=@n11, B=@n12), B=@n14)) {g=true, r=true}
  n25 = MaterialFunctionCall(In=Abs(None=ComponentMask(None=@n20.PS) {b=true}), Contrast=@n22) {materialFunction=&CheapContrast}
  n26 = LinearInterpolate(
   A=LinearInterpolate(
    A=TextureSample(UVs=@n16, Tex=@n17) {samplerType=SAMPLERTYPE_Masks, texture=/Game/Weapons/WeaponSkins/Textures/T_Swirly_Noise_highContrast}.RGB
    B=TextureSample(UVs=@n19, Tex=@n17) {samplerType=SAMPLERTYPE_Masks, texture=&T_Swirly_Noise_highContrast}.RGB
    Alpha=Saturate(None=@n23)
   )
   B=TextureSample(UVs=Add(A=@n24, B=@n15), Tex=@n17) {samplerType=SAMPLERTYPE_Masks, texture=&T_Swirly_Noise_highContrast}.RGB
   Alpha=Saturate(None=ComponentMask(None=@n25.Result) {b=true})
  )
  n27 = LinearInterpolate(
   A=ScalarParameter() {group=InvaderGlowLava, parameterName=LavaGradientBrightnessLow, sortPriority=31}
   B=ScalarParameter() {defaultValue=1, group=InvaderGlowLava, parameterName=LavaGradientBrightnessHigh}
   Alpha=Saturate(None=MaterialFunctionCall(Float3=@n26) {materialFunction=&BreakOutFloat3Components}.R)
  )
  n28 = Multiply(A=@n4.B, B=ScalarParameter() {defaultValue=1, group=InvaderGlowTextureMask, parameterName=B_LavaGradientOffset, sortPriority=30})
  n29 = TextureSampleParameter2D(UVs=Clamp(None=Subtract(A=@n27, B=@n28)) {maxDefault=0.99, minDefault=0.01}) {
   group=InvaderGlowLava
   parameterName=Lava UV Gradient
   samplerSource=SSM_Wrap_WorldGroupSettings
   sortPriority=12
   texture=/BRCosmetics/Characters/Player/Male/Large/Bodies/M_LRG_Invader/Skins/Glow/Textures/T_InvaderGlow_Lava_UV_Gradient
  }
  NamedRerouteDeclaration(None=Multiply(A=LinearInterpolate(B=@n9, Alpha=@n4.A) {constA=1}, B=Multiply(A=@n10.RGB, B=@n29.RGB))) {name=LavaEmissive, nodeColor=<rgba 0.5937 1 0 1>}
  n30 = OneMinus(None=SmoothStep(
    Min=ScalarParameter() {group=InvaderGlowGlass, parameterName=GlassMaskFresnelLow, sortPriority=13}
    Max=ScalarParameter() {defaultValue=1, group=InvaderGlowGlass, parameterName=GlassMaskFresnelHigh, sortPriority=14}
    Value=@n8
   ))
  n31 = ScalarParameter() {defaultValue=1, group=InvaderGlowGlass, parameterName=GlassMaskFresnelStrength, sortPriority=12}
  NamedRerouteDeclaration(None=Multiply(A=Multiply(A=@n30, B=@n31), B=@n4.A)) {name=FresnelMask, nodeColor=<rgba 0.1953 0 1 1>}
  n32 = VectorParameter() {group=InvaderGlowGlass, parameterName=GlassColor, sortPriority=10, defaultValue=<rgba 0.6667 0.9746 1 1>}
  n33 = Multiply(A=@n4.B, B=ScalarParameter() {group=InvaderGlowTextureMask, parameterName=B_GlassFresnelOffset, sortPriority=31})
  n34 = ScalarParameter() {group=InvaderGlowGlass, parameterName=GlassGradientFresnelOffset, sortPriority=22}
  n35 = ScalarParameter() {defaultValue=1, group=InvaderGlowGlass, parameterName=GlassGradientFresnelStrength, sortPriority=21}
  n36 = Fresnel(ExponentIn=ScalarParameter() {defaultValue=0.5, group=InvaderGlowGlass, parameterName=GlassGradientFresnelExponent, sortPriority=20}) {baseReflectFraction=0}
  n37 = TextureSampleParameter2D(UVs=Add(A=Add(A=@n33, B=@n34), B=Multiply(A=@n35, B=@n36))) {
   group=InvaderGlowGlass
   parameterName=GlassGradient
   sortPriority=11
   texture=/BRCosmetics/Characters/Player/Male/Large/Bodies/M_LRG_Invader/Skins/Glow/Textures/T_InvaderGlow_Glass_UV_Gradient
  }
  n38 = LinearInterpolate(
   A=@n32.RGB
   B=@n37.RGB
   Alpha=ScalarParameter() {defaultValue=0.5, group=InvaderGlowGlass, parameterName=GlassGradientFresnelIntensity, sortPriority=23}
  )
  NamedRerouteDeclaration(None=Multiply(
    A=NamedRerouteDeclaration(None=@n38) {name=GlassColor, nodeColor=<rgba 0 0.7656 1 1>}
    B=ScalarParameter() {group=InvaderGlowGlass, parameterName=GlassEmissiveStrength}
   )) {name=GlassEmissive, nodeColor=<rgba 0.1484 0 1 1>}
  n39 = NamedRerouteUsage()
  n40 = LinearInterpolate(B=@n4.R, Alpha=ScalarParameter() {defaultValue=1, group=InvaderGlowTextureMask, parameterName=R_GlassEdgeStrength, sortPriority=10})
  NamedRerouteDeclaration(None=Saturate(None=Add(A=@n39, B=@n40))) {name=Mask, nodeColor=<rgba 1 0.3047 0 1>}
  NamedRerouteDeclaration(None=@n40) {name=GlassEdges, nodeColor=<rgba 1 0.9375 0 1>}
  n41 = NamedRerouteUsage()
  n42 = NamedRerouteUsage()
  n43 = NamedRerouteUsage()
  n44 = NamedRerouteUsage()
  n45 = ScalarParameter() {defaultValue=1, group=InvaderGlowTextureMask, parameterName=R_GlassEdgeEmissiveStrength, sortPriority=11}
  NamedRerouteDeclaration(None=LinearInterpolate(A=LinearInterpolate(A=@n41, B=@n42, Alpha=@n43), B=Multiply(A=@n44, B=@n45), Alpha=NamedRerouteUsage())) {name=Emissive, nodeColor=<rgba 1 0 0.1641 1>}
 ]}
