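# Character skin material for the Invader large male Glow skin, using material attributes with a subsurface profile for skin.
# Character PreFX output passes through InvaderGlow and ThinFilm functions, each weighted by a selectable channel of the SkinFX_Mask texture.
# A static switch optionally multiplies emissive by panning scanlines on the second UV set, blended by another mask channel.
# The result then runs through shared character PostFX and cinematic PostFX functions.
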Material /BRCosmetics/Characters/Player/Male/Large/Bodies/M_LRG_Invader/Skins/Glow/Materials/M_Invader_Glow
graph {
 properties {
  bAutomaticallySetUsageInEditor=false
  bUseMaterialAttributes=true
  bUsedWithMeshDeformer=true
  bUsedWithMorphTargets=true
  bUsedWithSkeletalMesh=true
  pixelDepthOffsetMode=PDOM_Legacy
  shadingModel=MSM_FromMaterialExpression
  subsurfaceProfile=&SS_HeroSkin_02
  usedShadingModels=Default Lit | Subsurface Profile
 }
 nodes [
  GetMaterialAttributes()
  n1 = MaterialFunctionCall() {materialFunction=&BaseFN_CharacterFunction_PreFX}
  n2 = TextureSample(Tex=TextureObjectParameter() {group=Skin FX, parameterName=SkinFX_Mask})
  n3 = ChannelMaskParameter(None=@n2.RGB) {group=InvaderGlow, maskChannel=Green, parameterName=InvaderGlow_Channel, sortPriority=-1, defaultValue=<rgba 0 1 0 0>}
  n4 = GetMaterialAttributes(None=MaterialFunctionCall!(
    Alpha=ChannelMaskParameter(None=@n2.RGB) {group=ThinFilm, maskChannel=Blue, parameterName=ThinFilm_Channel, defaultValue=<rgba 0 0 1 0>}.R
    Material Attributes=MaterialFunctionCall!(Material=@n1.Result, Alpha=@n3.R) {materialFunction=&/MF_InvaderGlow}.Result
   ) {materialFunction=/Game/Weapons/WeaponSkins/MaterialFunction/ThinFilmFunction}.Result) {attributeGetTypes [#1]}
  n5 = TextureCoordinate() {coordinateIndex=1}
  n6 = LinearInterpolate(
   A=@n4.Emissive Color
   B=Multiply(
    A=@n4.Emissive Color
    B=MaterialFunctionCall!(ScanLine_Tiling_1, Scanline_PannerSpeed_1, ScanLine_Tiling, Scanline_PannerSpeed, UVs=@n5) {materialFunction=/Game/Weapons/WeaponSkins/MaterialFunction/ScanLines}.Result
   )
   Alpha=ChannelMaskParameter(None=@n2.RGB) {group=Scanlines, parameterName=Scanlines_Channel, sortPriority=-1}.R
  )
  n7 = MaterialFunctionCall(
   AO=@n1.AO
   Distance Blend=@n1.AO
   Fresnel=@n1.AO
   Gmap=@n1.AO
   Pre Skinned Local Position=@n1.AO
   Material=SetMaterialAttributes(
    MaterialAttributes=@n4.MaterialAttributes
    Emissive Color=StaticSwitchParameter(True=@n6, False=@n4.Emissive Color) {parameterName=UseScanlines}
   ) {attributeSetTypes [#1]}
  ) {materialFunction=&BaseFN_CharacterFunction_PostFX}
  MaterialFunctionCall(In=@n7.Result) {materialFunction=&CINE_MF_CharPostFX}
 ]
}
