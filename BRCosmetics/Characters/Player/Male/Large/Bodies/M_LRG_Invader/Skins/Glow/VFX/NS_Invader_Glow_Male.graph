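# Niagara system that adds red glow lights to the male Invader character's Glow skin.
# A single light-only emitter gates its low spawn rate on a comparison of User.IsGlow against zero.
# Two spawn groups index bones or sockets on User.SKM_03, resampled every update so lights follow the mesh.
# An alpha curve fades each particle in and out, with a camera offset applied and light attributes set in update.
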NiagaraSystem /BRCosmetics/Characters/Player/Male/Large/Bodies/M_LRG_Invader/Skins/Glow/VFX/NS_Invader_Glow_Male
dependencies {usedDataInterfaces [&NiagaraDataInterfaceColorCurve], usedDynamicInputs [], usedModules [&SystemState]}
emitters {ParticleLights {
  emitterUpdateScript [
   &EmitterState()
   &SpawnRate(
    SpawnRate=&MakeCustomFloatFromBool(Bool=&SetBoolByFloatComparison(A=$User.IsGlow, B=0), True Float=2, False Float=0)
    Use Spawn Probability=false
   )
  ]
  particleSpawnScript [
   /Niagara/Modules/Spawn/Initialization/InitializeParticle(
    Color=<rgba 1 0 0.0746 1>
    Module.Write Color=true
    Module.Write Lifetime=true
    Module.Write Mass=true
    Module.Write Material Random=true
    Module.Write Position=false
    Module.Write Scale=false
    Module.Write SpriteRotation=false
    Module.Write SpriteSize=true
    Sprite Size=&Vector2DFromFloat(Value=50)
   )
   &FN_SetSpawnGroup(Number of Spawn Groups=2)
   &SkeletalMeshLocation(
    Bone / Socket Index=&SpawnGroup()
    Bone Sampling Mode=::NewEnumerator4
    MaskSpawnGroup=false
    Skeletal Mesh=$User.SKM_03
   ) {module=SkeletalMeshLocation001}
  ]
  particleUpdateScript [
   &ParticleState()
   &CameraOffset(Camera Offset Amount=25)
   &SkeletalMeshLocation(
    Bone / Socket Index=&SpawnGroup()
    Bone Sampling Mode=::NewEnumerator4
    MaskSpawnGroup=false
    Skeletal Mesh=$User.SKM_03
   ) {module=SkeletalMeshLocation002}
   &ScaleColor(
    Scale Mode=::NewEnumerator1
    Scale RGBA=&Vector4FromCurve(Vector4Curve=DataInterface() {propertyValues {
       AlphaCurve {keys [<at 0 0>,<at 0.2467 0.85>,<at 0.8189 0.85>,<at 1 0>]}
       BlueCurve {keys [<at 0 1>,<at 1 1>]}
       GreenCurve {keys [<at 0 1>,<at 1 1>]}
       RedCurve {keys [<at 0 1>,<at 1 1>]}
      }})
    ScaleRGBA=true
   )
   &FN_Light_Attributes(
    Color=<rgba 1 0 0.0746 1>
    Create Color Attribute=true
    Create Exponent Attribute=true
    Create Radius Attribute=true
    Enable Particle Light Enable=true
    Exponent=3
    Light Enabled=true
    Radius=50
    Scale Color=3
   )
  ]
  rendererClasses [&NiagaraLightRendererProperties]
  renderers [{rendererClass=&NiagaraLightRendererProperties, rendererIndex=0}]
 }}
summary {
 systemName=NS_Invader_Glow_Male
 emitters [{emitterName=ParticleLights, rendererClasses [&NiagaraLightRendererProperties]}]
 userVariables [
  User.bIsFrontEndPreview {type=NiagaraBool}
  User.IsGlow {type=NiagaraFloat, defaultValue=0, description=""}
  User.SKM_03 {
   type=NiagaraDataInterfaceSkeletalMesh
   defaultValue {
    struct=NiagaraExt_VariableValue_DataInterface
    value {dataInterface=&self:NiagaraDataInterfaceSkeletalMesh_0}
   }
  }
 ]
}
